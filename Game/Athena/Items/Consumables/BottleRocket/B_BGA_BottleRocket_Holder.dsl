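; Thrown Bottle Rocket launcher that lands on non-pawn surfaces within a slope limit, aligns itself and fires rocket projectiles.
; PassVariables receives the ability's effect spec and player yaw; the server spawns rockets upward at randomized intervals, roughly one in five faster and tighter.
; If its base is destroyed, it rebinds to an overlapping building or resumes projectile movement.
; RepNotify variables sync resumed movement and muzzle-flash FX to clients.
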(event EventBeginPlay
  ;[1 1 1] Add hum audio component after a delay
  (Utilities|FlowControl|Delay (Variables|Default|GetInitial%20Delay))
  (AddComponent|NODE_AddAudioComponent-0) ; Looping Fire Cracker sound
)

(event Collision|EventHit
  (params (MyComp :type Primitive%20Component%20Object%20Reference) (Other :type Actor%20Object%20Reference) (OtherComp :type Primitive%20Component%20Object%20Reference) (bSelfMoved :type Boolean) (HitLocation :type Vector :default "0, 0, 0") (HitNormal :type Vector :default "0, 0, 0") (NormalImpulse :type Vector :default "0, 0, 0") (Hit :type Hit%20Result%20Structure%20%28by%20ref%29))
  ;[1 1 1] When this actor hits a flat enough surface, it stops simulating movement.
  (Utilities|Casting|CastToPawn Other
    (:CastFailed
      (if (< (Math|Trig|Acos(Degrees) (Math|Vector|DotProduct HitNormal "0, 0,1.000000")) (Variables|HitEvent|GetMaxSlope))
        (Variables|HitEvent|SetImpact_Loc HitLocation)
        (Variables|HitEvent|SetHitNormal HitNormal)
        (CallFunction|HandleRotation)
        (Game|Components|ProjectileMovement|StopSimulating (Variables|Default|GetProjectileMovement) Hit)
        (Networking|SwitchHasAuthority
          (:Authority
            (Variables|HitEvent|SetbResumeSimulation false)
            (CallFunction|StartSpawningBottleRockets)
            ;[1 1 1] Binds Event for when Building piece dies
            (bind _asbuilding_actor (Utilities|Casting|CastToBuildingActor Other)
              (:then
                (EventDispatchers|BindEventtoOnDied _asbuilding_actor (EventDispatchers|CreateEvent :!event Custom|BaseDied)))
              (:CastFailed
                (Game|BindEventtoOnDestroyed Other (EventDispatchers|CreateEvent :!event Custom|BaseDestroyed)))))
          (:Remote
            (Audio|PlaySoundatLocation (Variables|Default|GetPlacementCue) (Variables|HitEvent|GetImpact_Loc))))))))

(event Custom|BaseDied
  (params (DamagedActor :type Actor%20Object%20Reference) (Damage :type Float%20%28single-precision%29) (InstigatedBy :type Controller%20Object%20Reference) (DamageCauser :type Actor%20Object%20Reference) (HitLocation :type Vector) (FHitComponent :type Primitive%20Component%20Object%20Reference) (BoneName :type Name) (Momentum :type Vector))
  ;[1 1 1] Base destroyed. Start simulating again.
  (bind (_outactors _returnvalue) (Collision|SphereOverlapActors (Transformation|GetActorLocation) 16.0 (Utilities|Array|MakeArray "ObjectTypeQuery1") "/Script/FortniteGame.BuildingActor" (Utilities|Array|MakeArray self)))
  (if _returnvalue
    (bind _output_get (Variables|BaseEvents|SetFoundBuildingOnDied _outactors))
    (bind _asbuilding_actor (Utilities|Casting|CastToBuildingActor (Utilities|Array|Get(acopy) _output_get))
      (:then
        (EventDispatchers|BindEventtoOnDied _asbuilding_actor (EventDispatchers|CreateEvent :!event Custom|BaseDied)))
      (:CastFailed
        (Game|BindEventtoOnDestroyed _asbuilding_actor (EventDispatchers|CreateEvent :!event Custom|BaseDestroyed))))
    (else
      (Variables|HitEvent|SetbResumeSimulation true)
      (Networking|ForceNetUpdate))))

(event Custom|BaseDestroyed
  (params (DestroyedActor :type Actor%20Object%20Reference))
  ;[1 1 1] Base destroyed. Start simulating again.
  (bind (_outactors _returnvalue) (Collision|SphereOverlapActors (Transformation|GetActorLocation) 16.0 (Utilities|Array|MakeArray "ObjectTypeQuery1") "/Script/FortniteGame.BuildingActor" (Utilities|Array|MakeArray self)))
  (if _returnvalue
    (bind _output_get (Variables|BaseEvents|SetFoundBuildingOnDied _outactors))
    (bind _asbuilding_actor (Utilities|Casting|CastToBuildingActor (Utilities|Array|Get(acopy) _output_get))
      (:then
        (EventDispatchers|BindEventtoOnDied _asbuilding_actor (EventDispatchers|CreateEvent :!event Custom|BaseDied)))
      (:CastFailed
        (Game|BindEventtoOnDestroyed _asbuilding_actor (EventDispatchers|CreateEvent :!event Custom|BaseDestroyed))))
    (else
      (Variables|HitEvent|SetbResumeSimulation true)
      (Networking|ForceNetUpdate))))

(event Custom|StartSpawningBottleRockets
  ;[1 1 1] Spawn bottle rockets on a randomized interval
  (Utilities|FlowControl|Delay (Variables|SpawnEvent|GetDelayBottleRockets))
  (CallFunction|SpawnBottleRocket (Variables|SpawnEvent|GetEffectContainerSpecToApplyOnExplode))
  (CallFunction|TriggerMuzzleFlashFX)
  (Utilities|FlowControl|Delay (Math|Random|RandomFloatinRange (Variables|SpawnEvent|GetMin_BR_SpawnDelay) (Variables|SpawnEvent|GetMax_BR_SpawnDelay))))

(event Custom|PassVariables
  (params (EffectContainerSpecToApplyOnExplode :type Fort%20Gameplay%20Effect%20Container%20Spec%20Structure) (PlayerYaw :type Rotator))
  ;[1 1 1] Variables set by the ability spawning the bottle rocket launcher
  (Variables|SpawnEvent|SetEffectContainerSpecToApplyOnExplode EffectContainerSpecToApplyOnExplode)
  (Variables|HandleRotationEvent|SetMeshLandRotation PlayerYaw))

(fn HandleHitBuilding
  (params (Parent :type Scene%20Component%20Object%20Reference))
  (Transformation|AttachActorToComponent Parent "None" "KeepWorld" "KeepWorld" "KeepWorld" false))

(fn HandleRotation
  (bind _staticmesh (Variables|Default|GetStaticMesh))
  (Transformation|SetActorRotation (Math|Conversions|RotationFromXVector (Variables|HitEvent|GetHitNormal)))
  (Transformation|SetRelativeLocation _staticmesh (Variables|HandleRotationEvent|GetMeshLandLocationOffset) false false)
  (Transformation|SetRelativeRotation _staticmesh (Math|Rotator|MakeRotator (- (- (.yaw (Math|Conversions|RotationFromXVector (Variables|HitEvent|GetHitNormal))) (Variables|HandleRotationEvent|GetMeshLandYawOffset)) (.yaw (Variables|HandleRotationEvent|GetMeshLandRotation)))) false false))

(fn OnRep_bResumeSimulation
  (if (Variables|HitEvent|GetbResumeSimulation)
    (Components|Movement|SetUpdatedComponent (Variables|Default|GetProjectileMovement) (Variables|Default|GetCapsule))))

(fn OnRep_MuzzleFlashCounter
  (Networking|CanExecuteCosmeticEvents
    (:True
      (Components|Activation|Activate (Variables|Default|GetP_MuzzleFlash) true))))

(fn SpawnBottleRocket
  (params (EffectContainerSpecToApplyOnExplode :type Fort%20Gameplay%20Effect%20Container%20Spec%20Structure))
  ;[0 0.6816 1] 1/5 of rockets will be shot vertically higher (for flavor)
  (LocalVariables|SetHighShot (Utilities|Select :Index (Math|Random|RandomInteger 5) :Option%200 true :Option%201 false :Option%202 false :Option%203 false :Option%204 false))
  (Projectile|SpawnProjectile :ProjectileClass (Variables|SpawnBottleRocket|GetPrj_BottleRocket) :RequestedBy (Game|GetInstigator) :SpawnLocation (+ (Transformation|GetActorLocation) (* (Math|Vector|Normalize (Math|Rotator|GetRotationXVector (Transformation|GetActorRotation))) (Variables|SpawnBottleRocket|GetBR_SpawnLoc_Offset))) :SpawnRotation (Math|Conversions|RotationFromXVector (Math|Vector|MakeVector (+ (.x (Math|Vector|GetUpVector "0, 0, 0")) (Math|Random|RandomFloatinRange (select (LocalVariables|GetHighShot) (/ (Variables|SpawnBottleRocket|GetSpread) 3.0) (Variables|SpawnBottleRocket|GetSpread)) (* (select (LocalVariables|GetHighShot) (/ (Variables|SpawnBottleRocket|GetSpread) 3.0) (Variables|SpawnBottleRocket|GetSpread)) -1))) (+ (.y (Math|Vector|GetUpVector "0, 0, 0")) (Math|Random|RandomFloatinRange (select (LocalVariables|GetHighShot) (/ (Variables|SpawnBottleRocket|GetSpread) 3.0) (Variables|SpawnBottleRocket|GetSpread)) (* (select (LocalVariables|GetHighShot) (/ (Variables|SpawnBottleRocket|GetSpread) 3.0) (Variables|SpawnBottleRocket|GetSpread)) -1))) (+ (.z (Math|Vector|GetUpVector "0, 0, 0")) (Math|Random|RandomFloatinRange (select (LocalVariables|GetHighShot) (/ (Variables|SpawnBottleRocket|GetSpread) 3.0) (Variables|SpawnBottleRocket|GetSpread)) (* (select (LocalVariables|GetHighShot) (/ (Variables|SpawnBottleRocket|GetSpread) 3.0) (Variables|SpawnBottleRocket|GetSpread)) -1))))) :SpawnDirection (Math|Conversions|RotationFromXVector (Math|Vector|MakeVector (+ (.x (Math|Vector|GetUpVector "0, 0, 0")) (Math|Random|RandomFloatinRange (select (LocalVariables|GetHighShot) (/ (Variables|SpawnBottleRocket|GetSpread) 3.0) (Variables|SpawnBottleRocket|GetSpread)) (* (select (LocalVariables|GetHighShot) (/ (Variables|SpawnBottleRocket|GetSpread) 3.0) (Variables|SpawnBottleRocket|GetSpread)) -1))) (+ (.y (Math|Vector|GetUpVector "0, 0, 0")) (Math|Random|RandomFloatinRange (select (LocalVariables|GetHighShot) (/ (Variables|SpawnBottleRocket|GetSpread) 3.0) (Variables|SpawnBottleRocket|GetSpread)) (* (select (LocalVariables|GetHighShot) (/ (Variables|SpawnBottleRocket|GetSpread) 3.0) (Variables|SpawnBottleRocket|GetSpread)) -1))) (+ (.z (Math|Vector|GetUpVector "0, 0, 0")) (Math|Random|RandomFloatinRange (select (LocalVariables|GetHighShot) (/ (Variables|SpawnBottleRocket|GetSpread) 3.0) (Variables|SpawnBottleRocket|GetSpread)) (* (select (LocalVariables|GetHighShot) (/ (Variables|SpawnBottleRocket|GetSpread) 3.0) (Variables|SpawnBottleRocket|GetSpread)) -1))))) :EffectContainerSpecToApplyOnExplode EffectContainerSpecToApplyOnExplode :bUseDefaultPhysics false :InitialSpeed (select (LocalVariables|GetHighShot) (* (Math|Random|RandomFloatinRange (Variables|SpawnBottleRocket|GetBottleRocketSpeed_Min) (Variables|SpawnBottleRocket|GetBottleRocketSpeed_Max)) 2.0) (Math|Random|RandomFloatinRange (Variables|SpawnBottleRocket|GetBottleRocketSpeed_Min) (Variables|SpawnBottleRocket|GetBottleRocketSpeed_Max))) :GravityScale (Variables|SpawnBottleRocket|GetBottleRocketGravity) :ChargePercent 0.0)
  (Fortnite|PlayLocalForceFeedbackatLocationMulti (Variables|SpawnBottleRocket|GetForceFeedbackNear) (Variables|SpawnBottleRocket|GetForceFeedbackFar) (Transformation|GetActorLocation) 350 500 "BottleRocket"))

(fn TriggerMuzzleFlashFX
  (Variables|SpawnEvent|SetMuzzleFlashCounter (+ (Variables|SpawnEvent|GetMuzzleFlashCounter) 1)))

(fn ConstructionScript)
