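; Bottle rocket consumable projectile that plays a fire sound on spawn and explodes on impact or after a random lifetime.
; Bounces check the hit pawn's team affiliation before exploding, and both detonation paths clear bKillOnExplode before calling DoExplosion.
; On explosion it spawns FX and sound, shrinks and destroys the mesh, and deactivates the trail.
; The construction script gives the mesh a dynamic material instance.
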(event EventOnBounce
  (params (Hit :type Hit%20Result%20Structure%20%28by%20ref%29))
  (bind (_bblockinghit _binitialoverlap _time _distance _location _impactpoint _normal _impactnormal _physmat _hitactor) (Collision|BreakHitResult Hit))
  ;[1 1 1] Explode on Impact
  (bind _asfort_pawn (Utilities|Casting|CastToFortPawn _hitactor))
  (switch Utilities|FlowControl|Switch|SwitchonEFortTeamAffiliation (Team|GetTeamAffiliation _asfort_pawn self)
    (:Neutral
      _)
    (:Hostile
      _))
  (Variables|Delivery|SetbKillOnExplode false)
  (Delivery|DoExplosion))

(event EventBeginPlay
  (Audio|PlaySoundatLocation (Variables|Default|GetBottleRocketFireCue) (Variables|/Script/Engine.SceneComponent|GetRelativeLocation (Variables|Projectile_Firecracker_Mini|GetMesh)))
  (Collision|SetActorEnableCollision true)
  (Utilities|FlowControl|Delay (Math|Random|RandomFloatinRange (Variables|Default|GetLifetime_Min) (Variables|Default|GetLifetime_Max)))
  (Variables|Delivery|SetbKillOnExplode false)
  (Delivery|DoExplosion))

(event EventOnExploded
  (params (HitActors :type Array%20of%20Actor%20Object%20References) (HitResults :type Array%20of%20Hit%20Result%20Structures))
  (bind _returnvalue (Transformation|GetActorLocation))
  (bind _mesh (Variables|Projectile_Firecracker_Mini|GetMesh))
  (Effects|Components|ParticleSystem|SpawnFXAtLocation (Variables|Default|GetP_Explosion) _returnvalue (Transformation|GetActorRotation))
  (Audio|PlaySoundatLocation (Variables|Default|GetExplosionCue) _returnvalue)
  ;[1 1 1] ScaleDown Mesh and DeactivateFX
  (Transformation|SetWorldScale3D _mesh "0.010000,0.010000,0.010000")
  (Components|Activation|Deactivate (Variables|Projectile_Firecracker_Mini|GetP_Trail))
  (Components|DestroyComponent _mesh))

(fn ConstructionScript
  (bind _mesh (Variables|Projectile_Firecracker_Mini|GetMesh))
  (bind _returnvalue (Rendering|Material|CreateDynamicMaterialInstance _mesh 0 0 "None" :!class /Script/Engine.PrimitiveComponent))
  (Variables|Default|SetMine_Material _returnvalue)
  (Rendering|Material|SetMaterial _mesh 0 _returnvalue))
